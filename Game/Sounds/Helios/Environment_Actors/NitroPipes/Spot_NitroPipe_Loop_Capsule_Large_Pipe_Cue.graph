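# Ambient sound cue for a large NitroPipe spot loop, built from reused SewerPipes underground pipe waves.
# A Mixer layers randomly chosen, repeatedly delayed rattles with a heavily boosted pipe loop whose volume and pitch drift through a repeating Enveloper.
# All quality levels share that one mix.
# Overridden capsule attenuation with listener focus shapes falloff, a concurrency override limits instances, and a send feeds the exterior ambient source bus.
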SoundCue /Game/Sounds/Helios/Environment_Actors/NitroPipes/Spot_NitroPipe_Loop_Capsule_Large_Pipe_Cue
root=QualityLevel_0
nodes [
 {id=Delay_0, children [Random_0], properties {delayMax=5, delayMin=1}}
 {id=Delay_1, children [WavePlayer_10], properties {delayMax=0.5, delayMin=0.2}}
 {
  id=Enveloper_1
  children [Delay_1]
  properties {
   bLoopIndefinitely=true
   loopEnd=30
   pitchCurve {editorCurveData {keys [{interpMode=RCIM_Cubic, time=0}, <at 15 0.8 RCIM_Cubic>, {interpMode=RCIM_Cubic, time=30}]}}
   volumeCurve {editorCurveData {keys [<at 0 1 RCIM_Cubic>,<at 15 1.2 RCIM_Cubic>,<at 30 1 RCIM_Cubic>]}}
  }
 }
 {id=Looping_0, children [Delay_0]}
 {id=Mixer_1, children [Looping_0,Enveloper_1], properties {inputVolume [1,5]}}
 {id=QualityLevel_0, children [Mixer_1,Mixer_1,Mixer_1,Mixer_1]}
 {id=Random_0, children [WavePlayer_1,WavePlayer_3,WavePlayer_4,WavePlayer_8,WavePlayer_9], properties {weights [1,1,1,1,1]}}
 {id=WavePlayer_1, properties {soundWaveAssetPtr=/Game/Sounds/Helios/Environment_Actors/SewerPipes/Spot_UndergroundPipe_Rattle_01}}
 {
  id=WavePlayer_10
  properties {bLooping=true, soundWaveAssetPtr=/Game/Sounds/Helios/Environment_Actors/SewerPipes/Spot_UndergroundPipe_Loop_04}
 }
 {id=WavePlayer_3, properties {soundWaveAssetPtr=/Game/Sounds/Helios/Environment_Actors/SewerPipes/Spot_UndergroundPipe_Rattle_03}}
 {id=WavePlayer_4, properties {soundWaveAssetPtr=/Game/Sounds/Helios/Environment_Actors/SewerPipes/Spot_UndergroundPipe_Rattle_04}}
 {id=WavePlayer_8, properties {soundWaveAssetPtr=/Game/Sounds/Helios/Environment_Actors/SewerPipes/Spot_UndergroundPipe_Rattle_05}}
 {id=WavePlayer_9, properties {soundWaveAssetPtr=/Game/Sounds/Helios/Environment_Actors/SewerPipes/Spot_UndergroundPipe_Rattle_06}}
]
properties {
 bOverrideAttenuation=true
 bOverrideConcurrency=true
 maxDistance=1722.8429
 soundClassObject=&Fort_Ambient_Athena
 volumeMultiplier=1.002
 attenuationOverrides {
  attenuationShape=Capsule
  bEnableListenerFocus=true
  falloffDistance=958.9483
  nonFocusVolumeAttenuation=0.4
  nonSpatializedRadiusEnd=125
  nonSpatializedRadiusStart=150
  stereoSpread=150
  attenuationShapeExtents=<763.8946 440.1594>
 }
 busSends [{audioBus=null, minSendLevel=0.6, sendLevel=1, soundSourceBus=&AmbientAudio_Exterior_SourceBus}]
 concurrencyOverrides {maxCount=2}
 concurrencySet [&Weapon_Drop_StopOldest]
}
