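# Ambient loop cue for a large flowing nitro pipe, layering a steady flow bed with intermittent metallic bubbling.
# A looping wave runs through a repeating 30-second envelope that drifts pitch and volume.
# A quieter branch loops a randomly delayed, pitch-modulated pick of three bubble one-shots.
# All quality levels share this mix, with overridden capsule attenuation, listener focus, occlusion, limited concurrency and an exterior ambience bus send.
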SoundCue /Game/Sounds/Helios/Environment_Actors/NitroPipes/Spot_NitroPipe_Main_Large_Flowing_Loop_03_Short_Cue
root=QualityLevel_0
nodes [
 {id=Delay_0, children [Modulator_1], properties {delayMax=6, delayMin=1}}
 {
  id=Enveloper_1
  children [Modulator_2]
  properties {
   bLoopIndefinitely=true
   loopEnd=30
   pitchCurve {editorCurveData {keys [{interpMode=RCIM_Cubic, time=0}, <at 15 1.1 RCIM_Cubic>, {interpMode=RCIM_Cubic, time=30}]}}
   volumeCurve {editorCurveData {keys [<at 0 1 RCIM_Cubic>,<at 15 0.7 RCIM_Cubic>,<at 30 1 RCIM_Cubic>]}}
  }
 }
 {id=Looping_0, children [Delay_0]}
 {id=Mixer_1, children [Looping_0,Enveloper_1], properties {inputVolume [0.4,1]}}
 {id=Modulator_1, children [Random_1], properties {pitchMax=1.1, pitchMin=0.8}}
 {id=Modulator_2, children [WavePlayer_0]}
 {id=QualityLevel_0, children [Mixer_1,Mixer_1,Mixer_1,Mixer_1]}
 {id=Random_1, children [WavePlayer_6, WavePlayer_7, WavePlayer_10], properties {weights [1,1,1]}}
 {
  id=WavePlayer_0
  properties {bLooping=true, soundWaveAssetPtr=&/Spot_NitroPipe_Main_Large_Flowing_Loop_01}
 }
 {
  id=WavePlayer_10
  properties {soundWaveAssetPtr=&/Spot_NitroPipe_OneShot_Metallic_Bubbles_05_Stereo}
 }
 {
  id=WavePlayer_6
  properties {soundWaveAssetPtr=&/Spot_NitroPipe_OneShot_Metallic_Bubbles_03_Stereo}
 }
 {
  id=WavePlayer_7
  properties {soundWaveAssetPtr=&/Spot_NitroPipe_OneShot_Metallic_Bubbles_04_Stereo}
 }
]
properties {
 attenuationSettings=/Game/Sounds/Helios/Environment_Actors/SewerPipes/Spot_UndergroundPipe_Loop_Capsule_Att
 bEnableSubmixSends=false
 bOverrideAttenuation=true
 bOverrideConcurrency=true
 maxDistance=2168.4438
 soundClassObject=&Fort_Doors
 volumeMultiplier=0.3
 attenuationOverrides {
  attenuationShape=Capsule
  bEnableFocusInterpolation=true
  bEnableListenerFocus=true
  distanceAlgorithm=Logarithmic
  falloffDistance=1200
  focusAzimuth=25
  nonFocusAzimuth=45
  nonFocusVolumeAttenuation=0.2
  occlusionInterpolationTime=2
  occlusionTraceChannel=ECC_GameTraceChannel11
  occlusionVolumeAttenuation=0.6
  stereoSpread=50
  attenuationShapeExtents=<968.4438 200>
 }
 busSends [{audioBus=null, minSendLevel=0.6, sendLevel=1, soundSourceBus=&AmbientAudio_Exterior_SourceBus}]
 concurrencyOverrides {maxCount=2}
 concurrencySet [&Weapon_Drop_StopOldest]
}
