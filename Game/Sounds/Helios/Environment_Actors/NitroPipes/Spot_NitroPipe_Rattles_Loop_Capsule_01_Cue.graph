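# Looping ambient spot cue for nitro pipe rattles, built from two reused sewer pipe rattle waves.
# A random pick between the waves plays after a random delay, loops, and is pitched slightly down by a modulator.
# All quality levels share that one chain.
# Overrides give it capsule attenuation with listener focus and a tight concurrency cap, and it sends to the exterior ambient audio bus.
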SoundCue /Game/Sounds/Helios/Environment_Actors/NitroPipes/Spot_NitroPipe_Rattles_Loop_Capsule_01_Cue
root=QualityLevel_0
nodes [
 {id=Delay_0, children [Random_0], properties {delayMax=5, delayMin=0.5}}
 {id=Looping_0, children [Delay_0]}
 {id=Modulator_0, children [Looping_0], properties {pitchMax=0.9, pitchMin=0.8}}
 {id=QualityLevel_0, children [Modulator_0,Modulator_0,Modulator_0,Modulator_0]}
 {id=Random_0, children [WavePlayer_8,WavePlayer_9], properties {weights [1,1]}}
 {id=WavePlayer_8, properties {soundWaveAssetPtr=/Game/Sounds/Helios/Environment_Actors/SewerPipes/Spot_UndergroundPipe_Rattle_05}}
 {id=WavePlayer_9, properties {soundWaveAssetPtr=/Game/Sounds/Helios/Environment_Actors/SewerPipes/Spot_UndergroundPipe_Rattle_06}}
]
properties {
 attenuationSettings=/Game/Sounds/Helios/Environment_Actors/SewerPipes/Spot_UndergroundPipe_Loop_Capsule_Att
 bOverrideAttenuation=true
 bOverrideConcurrency=true
 maxDistance=1150
 soundClassObject=&Fort_Doors
 volumeMultiplier=1.2
 attenuationOverrides {
  attenuationShape=Capsule
  bEnableListenerFocus=true
  falloffDistance=850
  focusVolumeAttenuation=1.2
  nonFocusVolumeAttenuation=0.75
  nonSpatializedRadiusEnd=125
  nonSpatializedRadiusStart=150
  stereoSpread=150
  attenuationShapeExtents=<300 150>
 }
 busSends [{audioBus=null, minSendLevel=0.6, sendLevel=1, soundSourceBus=&AmbientAudio_Exterior_SourceBus}]
 concurrencyOverrides {maxCount=2}
 concurrencySet [&Weapon_Drop_StopOldest]
}
